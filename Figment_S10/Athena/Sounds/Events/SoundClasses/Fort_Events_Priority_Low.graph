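# Sound class asset for low-priority event audio in the Athena (Battle Royale) sounds of the Figment S10 content.
# It is a child of the Fort_Events_Master sound class, placing it within that events class hierarchy.
# All of its own properties stay at default values, so it serves as a grouping bucket for low-priority event sounds.
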SoundClass /Figment_S10/Athena/Sounds/Events/SoundClasses/Fort_Events_Priority_Low
parent=&/Fort_Events_Master
