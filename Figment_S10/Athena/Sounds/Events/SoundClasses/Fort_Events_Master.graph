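# Master sound class for event audio in the Figment_S10 Athena content, parented to the Fort_Athena_SFX class.
# It groups six child classes: a mapwide class, four priority tiers from extra high to low, and a Festivus class.
# The asset only defines this hierarchy, with its own class properties left at their usual settings.
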SoundClass /Figment_S10/Athena/Sounds/Events/SoundClasses/Fort_Events_Master
parent=/Figment_S10/Sounds/FortSoundClassesAndModes/Fort_Athena_SFX
children [
 &/Fort_Events_Mapwide
 &/Fort_Events_Priority_ExtHigh
 &/Fort_Events_Priority_High
 &/Fort_Events_Priority_Medium
 &/Fort_Events_Priority_Low
 &/Fort_Events_Festivus
]
